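# Niagara burst effect for the Crackle emote's snap: a very brief flash with streaks.
# Flash_Flare pairs a camera-offset sprite with a light, tinted by User.Ranked_Curve through rank-tier color curves (Unranked through Unreal).
# SpikeStreaks throws stretching, fading sparks under drag and gravity, while Corona_Burst adds two quickly expanding sprites that flash bright.
# User.GlobalScalar scales sizes, speeds, gravity and light radius in the active emitters; the Embers emitter is disabled.
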NiagaraSystem /BRCosmetics/Effects/Fort_Effects/Emotes/Crackle/NS_Emote_Crackle_Snap
dependencies {
 usedDataInterfaces [&NiagaraDataInterfaceColorCurve, &NiagaraDataInterfaceCurve, &NiagaraDataInterfaceVector2DCurve]
 usedDynamicInputs []
 usedModules [&SystemState]
}
emitters {
 Corona_Burst {
  emitterUpdateScript [
   &EmitterState(Life Cycle Mode=::NewEnumerator1, Loop Behavior=::NewEnumerator1, Loop Duration=0.1, UseLoopDelay=false)
   &SpawnBurst_Instantaneous(Spawn Count=2, Use Spawn Probability=false, UseLoopCountLimit=false)
  ]
  particleSpawnScript [&InitializeParticle(
    Color=<rgba 1 1 1 0.5>
    Lifetime=0.1
    Sprite Rotation Mode=::NewEnumerator2
    Sprite Size Mode=::NewEnumerator3
    Uniform Sprite Size=&Multiply_Float(A=120, B=$User.GlobalScalar)
    UsePositionOffset=false
    Write Material Random=true
   )]
  particleUpdateScript [
   &ParticleState()
   &ScaleSpriteSize(Uniform Curve Sprite Scale=DataInterface() {propertyValues {Curve {keys [
        {interpMode=RCIM_Cubic, leaveTangentWeight=0.9682, tangentMode=RCTM_User, time=0, value=0}
        {arriveTangent=0.3352, arriveTangentWeight=0.3333, interpMode=RCIM_Cubic, leaveTangent=0.3352, tangentMode=RCTM_User, time=1, value=1}
       ]}}})
   &DynamicMaterialParameters(
    Index 0 Param 1=&FloatFromCurve(FloatCurve=DataInterface() {propertyValues {Curve {keys [<at 0.8 1 RCIM_Cubic>, {arriveTangent=-5.1423, interpMode=RCIM_Cubic, leaveTangent=-5.1423, tangentMode=RCTM_User, time=1, value=0}]}}})
    Index 0 Param 2=1
    Index 0 Param 3=1
    Index 0 Param 4=1
    Module.Param0WriteEnabled=true
    Module.Param1WriteEnabled=true
    Module.Param2WriteEnabled=true
    Module.Param3WriteEnabled=true
   )
   &ScaleColor(
    Scale Alpha=&FloatFromCurve(FloatCurve=DataInterface() {propertyValues {Curve {keys [{arriveTangent=-5.8164, interpMode=RCIM_Cubic, leaveTangent=-5.8164, tangentMode=RCTM_User, time=0.75, value=1}, <at 1 0 RCIM_Cubic>]}}})
    Scale RGB=&VectorFromFloat(Value=&FloatFromCurve(FloatCurve=DataInterface() {propertyValues {Curve {keys [{arriveTangent=-62.717, interpMode=RCIM_Cubic, leaveTangent=-62.717, tangentMode=RCTM_User, time=0, value=10}, <at 0.3 1 RCIM_Cubic>]}}}))
    ScaleA=true
    ScaleRGB=true
   )
  ]
  rendererClasses [&NiagaraSpriteRendererProperties]
  renderers [{rendererClass=&NiagaraSpriteRendererProperties}]
 }
 Embers {
  bEnabled=false
  emitterUpdateScript [
   &EmitterState(Life Cycle Mode=::NewEnumerator1, Loop Behavior=::NewEnumerator1, Loop Duration=0.8, UseLoopDelay=false)
   &SpawnRate(
    SpawnRate=&FloatFromCurve(FloatCurve=DataInterface() {propertyValues {Curve {keys [<at 0 100>,<at 0.15 100>,<at 0.2 0>]}}}, CurveIndex=$Emitter.NormalizedLoopAge)
    Use Spawn Probability=false
   )
   &SpawnBurst_Instantaneous(
    Spawn Count=&RandomRangeInt(Minimum=7, Maximum=10)
    Use Spawn Probability=false
    UseLoopCountLimit=false
   )
  ]
  particleSpawnScript [
   &InitializeParticle(
    Lifetime Max=0.85
    Lifetime Min=0.25
    Lifetime Mode=::NewEnumerator1
    Sprite Rotation Mode=::NewEnumerator2
    Sprite Size Mode=::NewEnumerator1
    Uniform Sprite Size Max=2
    Uniform Sprite Size Min=1
    UsePositionOffset=false
    Write Material Random=true
   )
   &ShapeLocation(MaskSpawnGroup=false, Override Seed=false, Sphere Radius=5, UseOwnerScale=true)
   &AddVelocity(
    Module.Use Falloff Distance=false
    Override Seed=false
    Velocity Mode=::NewEnumerator1
    Velocity Speed=&RandomRangeFloat(Minimum=150, Maximum=300)
   )
   &AddVelocity(Velocity=&RandomRangeVector(Minimum=<15 100 25>, Maximum=<15 100 75>)) {module=AddVelocity001}
  ]
  particleUpdateScript [
   &ParticleState()
   &ScaleSpriteSize(Uniform Curve Sprite Scale=DataInterface() {propertyValues {Curve {keys [
        {arriveTangent=9.1531, interpMode=RCIM_Cubic, leaveTangent=9.1531, tangentMode=RCTM_User, time=0, value=0.02}
        {arriveTangent=0.48, interpMode=RCIM_Cubic, leaveTangent=0.48, time=0.18, value=0.99}
        <at 1 0.5 RCIM_Cubic>
       ]}}})
   &AccelerationForce(Acceleration=<0 0 250>)
   &CurlNoiseForce(
    Noise Frequency=0.0125
    Noise Strength=&FloatFromCurve(FloatCurve=DataInterface() {propertyValues {Curve {keys [<at 0 1>,<at 1 0>]}}}, Scale Curve=1000)
    Pan Noise=true
    Pan Noise Field=<0 0 1>
   )
   &Drag(Drag=5, Use Linear Drag=true, Use Rotational Drag=true)
   &SolveForcesAndVelocity(Clamp Velocity=false, Limit Acceleration=false)
   &ScaleSpriteSizeBySpeed(Max Scale Factor=<1 4>, Min Scale Factor=<1 2>)
   &DynamicMaterialParameters(
    Index 0 Param 1=&FloatFromCurve(FloatCurve=DataInterface() {propertyValues {Curve {keys [<at 0 1>,<at 1 2>]}}})
    Index 0 Param 2=5
    Index 0 Param 3=1
    Index 0 Param 4=1
    Module.Param0WriteEnabled=true
    Module.Param1WriteEnabled=true
    Module.Param2WriteEnabled=true
    Module.Param3WriteEnabled=true
   )
   &ScaleColor(
    Scale Alpha=&FloatFromCurve(FloatCurve=DataInterface() {propertyValues {Curve {keys [
         {arriveTangent=11.9433, interpMode=RCIM_Cubic, leaveTangent=11.9433, tangentMode=RCTM_User, time=0, value=0}
         <at 0.2 1 RCIM_Cubic>
         <at 1 0 RCIM_Cubic>
        ]}}})
    Scale RGB=&MakeVectorFromLinearColorRGB(Color=&ColorFromCurve(ColorCurve=DataInterface() {propertyValues {
        AlphaCurve {keys []}
        BlueCurve {keys [<at 0 5 RCIM_Cubic>,<at 1 0.5 RCIM_Cubic>]}
        GreenCurve {keys [<at 0 5 RCIM_Cubic>,<at 1 0.5 RCIM_Cubic>]}
        RedCurve {keys [<at 0 5 RCIM_Cubic>,<at 1 0.5 RCIM_Cubic>]}
       }}))
    ScaleA=true
    ScaleRGB=true
   )
  ]
  rendererClasses [&NiagaraSpriteRendererProperties]
  renderers [{rendererClass=&NiagaraSpriteRendererProperties}]
 }
 Flash_Flare {
  emitterUpdateScript [
   &EmitterState()
   &SpawnBurst_Instantaneous(Use Spawn Probability=false, UseLoopCountLimit=false)
  ]
  particleSpawnScript [
   &InitializeParticle(
    Color=&self:SelectColorCurve_01(
     CurveSelect:NiagaraFloat=$User.Ranked_Curve
     ColorSelect:NiagaraFloat=0.5
     UnrankedColorCurve:NiagaraDataInterfaceColorCurve=DataInterface() {propertyValues {
       AlphaCurve {keys [<at 0 1>,<at 1 1>]}
       BlueCurve {keys [
         {interpMode=RCIM_Cubic, leaveTangentWeight=0.4171, tangentMode=RCTM_User, time=0, value=0.2831}
         {arriveTangent=1.9677, interpMode=RCIM_Cubic, leaveTangent=1.9677, tangentMode=RCTM_User, time=1, value=1}
        ]}
       GreenCurve {keys [
         {interpMode=RCIM_Cubic, leaveTangentWeight=0.4171, tangentMode=RCTM_User, time=0, value=0.2831}
         {arriveTangent=1.9677, interpMode=RCIM_Cubic, leaveTangent=1.9677, tangentMode=RCTM_User, time=1, value=1}
        ]}
       RedCurve {keys [
         {interpMode=RCIM_Cubic, leaveTangentWeight=0.4171, tangentMode=RCTM_User, time=0, value=0.2831}
         {arriveTangent=1.9677, interpMode=RCIM_Cubic, leaveTangent=1.9677, tangentMode=RCTM_User, time=1, value=1}
        ]}
      }}
     BronzeColorCurve:NiagaraDataInterfaceColorCurve=DataInterface() {propertyValues {
       AlphaCurve {keys [<at 0 1>,<at 1 1>]}
       BlueCurve {keys [<at 0 0.0185>,<at 1 0.1946>]}
       GreenCurve {keys [<at 0 0.0714>,<at 1 0.3557>]}
       RedCurve {keys [<at 0 0.2831>,<at 1 1>]}
      }}
     SilverColorCurve:NiagaraDataInterfaceColorCurve=DataInterface() {propertyValues {
       AlphaCurve {keys [<at 0 1>,<at 1 1>]}
       BlueCurve {keys [<at 0 0.0409>,<at 1 0.9734>]}
       GreenCurve {keys [<at 0 0.0273>,<at 1 0.9301>]}
       RedCurve {keys [<at 0 0.0194>,<at 1 0.8796>]}
      }}
     GoldColorCurve:NiagaraDataInterfaceColorCurve=DataInterface() {propertyValues {
       AlphaCurve {keys [<at 0 1>,<at 1 1>]}
       BlueCurve {keys [<at 0 0>,<at 1 0.2961>]}
       GreenCurve {keys [<at 0 0.1694>,<at 1 0.7654>]}
       RedCurve {keys [<at 0 0.2542>,<at 1 1>]}
      }}
     PlatColorCurve:NiagaraDataInterfaceColorCurve=DataInterface() {propertyValues {
       AlphaCurve {keys [<at 0 1>,<at 1 1>]}
       BlueCurve {keys [<at 0 0.0221>,<at 1 0.404>]}
       GreenCurve {keys [<at 0 0.05>,<at 1 0.9216>]}
       RedCurve {keys [<at 0 0.0021>,<at 1 0.0343>]}
      }}
     DiamondColorCurve:NiagaraDataInterfaceColorCurve=DataInterface() {propertyValues {
       AlphaCurve {keys [<at 0 1>,<at 1 1>]}
       BlueCurve {keys [<at 0 0.4>,<at 1 1>]}
       GreenCurve {keys [<at 0 0.2356>,<at 1 0.5833>]}
       RedCurve {keys [<at 0 0.0054>,<at 1 0>]}
      }}
     EliteColorCurve:NiagaraDataInterfaceColorCurve=DataInterface() {propertyValues {
       AlphaCurve {keys [<at 0 1>,<at 1 1>]}
       BlueCurve {keys [<at 0 0.0005 RCIM_Cubic>, {arriveTangent=0.3468, interpMode=RCIM_Cubic, leaveTangent=0.3468, tangentMode=RCTM_User, time=1, value=0.1}]}
       GreenCurve {keys [<at 0 0.0005 RCIM_Cubic>, {arriveTangent=0.3468, interpMode=RCIM_Cubic, leaveTangent=0.3468, tangentMode=RCTM_User, time=1, value=0.1}]}
       RedCurve {keys [<at 0 0.0005 RCIM_Cubic>, {arriveTangent=0.3468, interpMode=RCIM_Cubic, leaveTangent=0.3468, tangentMode=RCTM_User, time=1, value=0.1}]}
      }}
     ChampionColorCurve:NiagaraDataInterfaceColorCurve=DataInterface() {propertyValues {
       AlphaCurve {keys [<at 0 1>,<at 1 1>]}
       BlueCurve {keys [<at 0 0>,<at 0.5 0>,<at 1 0.0012>]}
       GreenCurve {keys [<at 0 0 RCIM_Cubic>, {arriveTangent=0.6654, interpMode=RCIM_Cubic, leaveTangent=0.6654, time=0.5, value=0.0908}, <at 1 0.6654 RCIM_Cubic>]}
       RedCurve {keys [
         <at 0 0.15 RCIM_Cubic>
         {arriveTangentWeight=0.2187, interpMode=RCIM_Cubic, leaveTangentWeight=0.2187, tangentMode=RCTM_User, time=0.5, value=1}
         <at 1 1 RCIM_Cubic>
        ]}
      }}
     UnrealColorCurve:NiagaraDataInterfaceColorCurve=DataInterface() {propertyValues {
       AlphaCurve {keys [<at 0 1>,<at 1 1>]}
       BlueCurve {keys [<at 0 0.2307 RCIM_Cubic>, {interpMode=RCIM_Cubic, leaveTangent=0.5013, time=1, value=0.2664}]}
       GreenCurve {keys [<at 0 0.0252 RCIM_Cubic>, {interpMode=RCIM_Cubic, leaveTangent=0.6649, time=1, value=0.0782}]}
       RedCurve {keys [
         <at 0 0.0222 RCIM_Cubic>
         {arriveTangentWeight=0.2187, interpMode=RCIM_Cubic, leaveTangent=0.4814, leaveTangentWeight=0.2187, time=1, value=0.3564}
        ]}
      }}
    )
    Lifetime=0.1
    Sprite Size Mode=::NewEnumerator3
    Uniform Sprite Size=&Multiply_Float(A=240, B=$User.GlobalScalar)
    UsePositionOffset=false
    Write Material Random=true
   )
   &DynamicMaterialParameters(
    Index 0 Param 1=0.2
    Index 0 Param 2=1
    Module.Param0WriteEnabled=true
    Module.Param1WriteEnabled=true
    Module.Param2WriteEnabled=true
    Module.Param3WriteEnabled=false
   )
   &CameraOffset(Camera Offset Amount=&Multiply_Float(A=50, B=$User.GlobalScalar))
  ]
  particleUpdateScript [
   &ParticleState()
   &ScaleColor(
    Scale Alpha=&FloatFromCurve(FloatCurve=DataInterface() {propertyValues {Curve {keys [<at 0 0.1>,<at 1 0>]}}})
    Scale RGB=&VectorFromFloat(Value=8)
    ScaleA=true
    ScaleRGB=true
   )
   /SetVariables(
    Particles.LightColor=&MultiplyLinearColor(
     Linear Color=$Particles.Color
     Scale Factor=&MakeFloat4FromVectorAndFloat("Vector (RGB)"=<1 1 1>, "Float (Alpha)"=10)
    )
    Particles.LightExponent=3
    Particles.LightRadius=&Multiply_Float(A=150, B=$User.GlobalScalar)
   ) {isSetParameters=true}
  ]
  rendererClasses [&NiagaraSpriteRendererProperties, &NiagaraLightRendererProperties]
  renderers [
   {rendererClass=&NiagaraSpriteRendererProperties}
   {rendererClass=&NiagaraLightRendererProperties, rendererIndex=1}
  ]
 }
 SpikeStreaks {
  emitterUpdateScript [
   &EmitterState()
   &SpawnBurst_Instantaneous(
    Spawn Count=&RandomRangeInt(Minimum=12, Maximum=12)
    Use Spawn Probability=false
    UseLoopCountLimit=false
   )
  ]
  particleSpawnScript [
   &InitializeParticle(
    Lifetime Max=0.15
    Lifetime Min=0.075
    Lifetime Mode=::NewEnumerator1
    Sprite Size Mode=::NewEnumerator1
    Uniform Sprite Size Max=&Multiply_Float(A=8, B=$User.GlobalScalar)
    Uniform Sprite Size Min=&Multiply_Float(A=5, B=$User.GlobalScalar)
    UsePositionOffset=false
    Write Material Random=true
   )
   &ShapeLocation(
    MaskSpawnGroup=false
    Override Seed=false
    Sphere Radius=&Multiply_Float(A=2, B=$User.GlobalScalar)
    UseOwnerScale=true
   )
   &AddVelocity(
    Module.Use Falloff Distance=false
    Override Seed=false
    Velocity Mode=::NewEnumerator1
    Velocity Speed=&Multiply_Float(A=&RandomRangeFloat(Minimum=350, Maximum=1000), B=$User.GlobalScalar)
   )
  ]
  particleUpdateScript [
   &ParticleState()
   &ScaleSpriteSize(
    Non-Uniform Curve Sprite Scale=DataInterface() {propertyValues {
      XCurve {keys [
        {arriveTangent=5.9728, interpMode=RCIM_Cubic, leaveTangent=5.9728, tangentMode=RCTM_User, time=0, value=0.1}
        {arriveTangent=-0.1, interpMode=RCIM_Cubic, leaveTangent=-0.1, tangentMode=RCTM_User, time=0.18, value=0.8}
        {interpMode=RCIM_Cubic, tangentMode=RCTM_User, time=1, value=0}
       ]}
      YCurve {keys [
        {arriveTangent=6.2715, interpMode=RCIM_Cubic, leaveTangent=6.2715, tangentMode=RCTM_User, time=0, value=0.1}
        {arriveTangent=8.8899, interpMode=RCIM_Cubic, leaveTangent=8.8899, tangentMode=RCTM_User, time=0.1788, value=2.5067}
        {arriveTangent=1.9556, interpMode=RCIM_Cubic, leaveTangent=1.9556, tangentMode=RCTM_User, time=0.9988, value=7}
       ]}
     }}
    Scale Sprite Size Mode=::NewEnumerator3
   )
   &ScaleColor(
    Scale Alpha=&FloatFromCurve(FloatCurve=DataInterface() {propertyValues {Curve {keys [<at 0.4 1.0113 RCIM_Cubic>,<at 1.0011 0.002>]}}})
    Scale RGB=&MakeVectorFromLinearColorRGB(Color=&ColorFromCurve(ColorCurve=DataInterface() {propertyValues {
        AlphaCurve {keys []}
        BlueCurve {keys [<at 0 1.5 RCIM_Cubic>,<at 0.2 30>,<at 0.627 5>,<at 1 0.5>]}
        GreenCurve {keys [<at 0 1.5 RCIM_Cubic>,<at 0.2 30>,<at 0.627 5>,<at 1 0.5>]}
        RedCurve {keys [<at 0 1.5 RCIM_Cubic>,<at 0.2 30>,<at 0.627 5>,<at 1 0.5>]}
       }}))
    ScaleA=true
    ScaleRGB=true
   )
   &GravityForce(Gravity=&MakeVector(X=0, Y=0, Z=&Multiply_Float(A=-250, B=$User.GlobalScalar)))
   &Drag(Drag=&RandomRangeFloat(Minimum=2, Maximum=6), Use Linear Drag=true, Use Rotational Drag=true)
   &SolveForcesAndVelocity(Clamp Velocity=false, Limit Acceleration=false)
  ]
  rendererClasses [&NiagaraSpriteRendererProperties]
  renderers [{rendererClass=&NiagaraSpriteRendererProperties}]
 }
}
summary {
 systemName=NS_Emote_Crackle_Snap
 emitters [
  {emitterName=Flash_Flare, rendererClasses [&NiagaraSpriteRendererProperties, &NiagaraLightRendererProperties]}
  {emitterName=SpikeStreaks, rendererClasses [&NiagaraSpriteRendererProperties]}
  {emitterName=Corona_Burst, rendererClasses [&NiagaraSpriteRendererProperties]}
  {bEnabled=false, emitterName=Embers, rendererClasses [&NiagaraSpriteRendererProperties]}
 ]
 userVariables [
  User.GlobalScalar {type=NiagaraFloat, defaultValue=0.25, description=""}
  User.Ranked_Curve {type=NiagaraFloat, defaultValue=0, description=""}
  User.Unranked {type=NiagaraBool}
 ]
}
